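# Niagara system for the NumbLTE03 vehicle boost cosmetic, trailing blue scarves and snow from the car's boost sockets.
# Glow bursts persistent sprites per port, while Snow spawns two groups of rearward-drifting sprites.
# Scarf_L and Scarf_R are three-renderer ribbons waved by age-driven sines, with velocity switched by game mode.
# A frontend-preview flag adjusts brightness, snow speed and glow offset; painted-variant overrides are disabled.
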NiagaraSystem /VehicleCosmetics/Boosts/Boost_NumbLTE03/FX/NS_NumbLTE03
dependencies {
 usedDataInterfaces [&NiagaraDataInterfaceCurve, &NiagaraDataInterfaceArrayFloat3, &NiagaraDataInterfaceSkeletalMesh]
 usedDynamicInputs []
 usedModules [
  &self:SystemSpawnScript.SystemScriptSource.SystemScriptGraph.NiagaraNodeAssignment_0.SetVariables_#1
  &SystemState
 ]
}
emitters {
 Glow {
  emitterUpdateScript [
   &EmitterState(
    Inactive Response=::NewEnumerator1
    Life Cycle Mode=::NewEnumerator1
    Loop Behavior=::NewEnumerator1
    Loop Duration Mode=::NewEnumerator1
    UseLoopDelay=false
   )
   &VC_FilteredSocketCount()
   &SpawnRate(SpawnRate=30, Use Spawn Probability=false) {enabled=false}
   &SpawnBurst_Instantaneous(Spawn Count=&Multiply_Int(A=$Emitter.PortCount, B=2), Use Spawn Probability=false, UseLoopCountLimit=false)
  ]
  particleSpawnScript [
   &VC_PaintedVariantsOverride() {enabled=false}
   &InitializeParticle(
    Color=&MergeHelper_ColorSwitch(
     0=<rgba 0.0043 0.0608 0.495 0.6>
     1=<rgba 0.0087 0.1217 0.9913 0.6>
     2=?
     3=?
     Index=&MakeCustomIntFromBool(Bool=$User.bIsFrontendPreview, True INT=1, False INT=0)
     Num=::NewEnumerator0
    )
    Lifetime=0.25
    Sprite Size=<370 70>
    Sprite Size Mode=::NewEnumerator4
    UsePositionOffset=false
    Write Material Random=true
   )
   &PartitionParticles(Number Of Partitions=$Emitter.PortCount)
   &SkeletalMeshLocation(
    Bone / Socket Index=$Particles.PartitionParticles.Partition
    Bone / Socket Sampling Mode=::NewEnumerator5
    MaskSpawnGroup=false
    Mesh Sampling Type=::NewEnumerator6
    Skeletal Mesh=$System.SkeletalMesh
   )
   &AddVelocity(Velocity=<-1500 0 0>) {enabled=false}
   /SetVariables(Particles.flarecolor:LinearColor=<rgba 0.3043 0.4795 0.7304 0.65>, Particles.flaresize:Vector2f=<120 90>) {isSetParameters=true}
   /SetVariables(Particles.Position=&AddVectorToPosition(
     Position=$Particles.Position
     Vector=&SelectVectorFromArray(
      Direct Array Index=&MakeCustomIntFromBool(Bool=$User.bIsFrontendPreview, True INT=1, False INT=0)
      Vector Selection Array=DataInterface() {propertyValues {FloatData [<75 0 0>, <0 0 0>]}}
     )
    )) {isSetParameters=true}
   &CameraOffset(Camera Offset Amount=60)
  ]
  particleUpdateScript [
   &ParticleState(Kill Particles When Lifetime Has Elapsed=false)
   &SolveForcesAndVelocity(Clamp Velocity=false, Limit Acceleration=false) {enabled=false}
  ]
  rendererClasses [&NiagaraSpriteRendererProperties]
  renderers [
   {rendererClass=&NiagaraSpriteRendererProperties}
   {rendererClass=&NiagaraSpriteRendererProperties, rendererIndex=1}
  ]
 }
 Scarf_L {
  emitterUpdateScript [
   &EmitterState()
   &VC_FilteredSocketCount()
   &VC_RateQualityHandler(Epic Spawn Rate=40, High Spawn Rate=40, Low Spawn Rate=30, Use Custom Rate?=true)
   &VC_SpawnRibbon_v1(
    Number of Ribbons=/Niagara/DynamicInputs/Subtract/Subtract_Integer(A:NiagaraInt32=$Emitter.PortCount, B:NiagaraInt32=1)
    SpawnRate=$Emitter.Spawn Rate
   )
  ]
  particleSpawnScript [
   &VC_PaintedVariantsOverride() {enabled=false}
   &VC_InitializeRibbon_v1(
    Boneflow Offset=1
    Custom Offset=true
    Enable Skeletal Position=true
    Facing Axis:ENiagaraOrientationAxis=::YAxis
    Flip Facing=true
    Lifetime=0.8
    Local Vector Offset=&AddVectorToPosition(Position=<20 10 0>, Vector=$User.DTCarOffset)
    Offset Amount=1
    Ribbon Width=30
   )
   &AddVelocity(Velocity=&MergeHelper_VectorSwitch(0=<-1500 0 0>, 1=<-1500 0 0>, 2=<-1200 0 0>, 3=?, Index=$User.GameModeID, Num=::NewEnumerator1)) {module=AddVelocity001}
   /SetVariables() {isSetParameters=true, module=SetVariables_#2}
   /SetVariables(Particles.Position=&AddVectorToPosition(
     Position=$Particles.Position
     Vector=&TransformVector(
      Vector=&MakeVector(X=0, Y=&MaskFloatBySpawnGroup(Masked Float=10, Passthrough Float=-10), Z=0)
      Destination Space=::Simulation
     )
    )) {isSetParameters=true, module=SetVariables_#3}
  ]
  particleUpdateScript [
   &ParticleState()
   &InheritSourceMovement(ApplyForce=false, ApplyPosition=true, ApplyPositionFalloff=false)
   &SolveForcesAndVelocity(Clamp Velocity=false, Limit Acceleration=false)
   &VC_UpdateRibbon_v1(Detach Locked Particle=true, Enable Skeletal Position=true)
   /SetVariables(Particles.Position=&AddVectorToPosition(
     Position=$Particles.Position
     Vector=/VehicleCosmetics/FX/Niagara/Modules/VC_ConvertVectorToWorld(Vector:Vector3f=&MakeVector(
       X=0
       Y=0
       Z=&FloatFromCurve(
        FloatCurve=DataInterface() {propertyValues {Curve {keys [
            <at 0 0 RCIM_Cubic>
            {arriveTangent=103.5191, interpMode=RCIM_Cubic, leaveTangent=103.5191, tangentMode=RCTM_User, time=0.22, value=40}
            {arriveTangent=3.1163, interpMode=RCIM_Cubic, leaveTangent=3.1163, tangentMode=RCTM_User, time=0.5, value=40}
            {arriveTangent=-14.8689, interpMode=RCIM_Cubic, leaveTangent=-14.8689, tangentMode=RCTM_User, time=0.8, value=0}
           ]}}}
        Scale Curve=0.05
       )
      ))
    )) {isSetParameters=true, module=SetVariables_#4}
   /SetVariables(Particles.Position=&AddVectorToPosition(
     Position=$Particles.Position
     Vector=&TransformVector(
      Vector=&MakeVector(
       X=0
       Y=&Sine(
        Normalized Angle=&Add_Float(
         A=&Multiply_Float(A=$Particles.NormalizedAge, B=-1.2)
         B=&Multiply_Float(A=$Emitter.LoopedAge, B=1)
        )
        Period=1
        Scale=&FloatFromCurve(
         FloatCurve=DataInterface() {propertyValues {Curve {defaultValue=1, keys [{tangentMode=RCTM_User, time=0, value=0}, {tangentMode=RCTM_User, time=0.53, value=1.1}]}}}
         Scale Curve=6
        )
        Bias=0
       )
       Z=&Sine(
        Normalized Angle=&Add_Float(
         A=&Multiply_Float(A=$Particles.NormalizedAge, B=-1.3)
         B=&Multiply_Float(A=$Emitter.LoopedAge, B=1)
        )
        Period=1
        Scale=&FloatFromCurve(
         FloatCurve=DataInterface() {propertyValues {Curve {
            defaultValue=1
            keys [
             {tangentMode=RCTM_User, time=0.1, value=0.3}
             {arriveTangent=-0.2778, interpMode=RCIM_Cubic, leaveTangent=-3.8301, tangentMode=RCTM_Break, time=0.3, value=1.2}
             <at 1 0.05 RCIM_Cubic>
            ]
           }}}
         Scale Curve=4
        )
        Bias=0
       )
      )
      Destination Space=::Simulation
      Apply Scale=true
     )
    )) {isSetParameters=true, module=SetVariables_#5}
   /SetVariables(Particles.RibbonFacing=&MaskVectorBySpawnGroup(
     Masked Vector=&MakeVector(
      X=1
      Y=0
      Z=&Sine(
       Normalized Angle=&Add_Float(
        A=&Multiply_Float(A=$Particles.NormalizedAge, B=1)
        B=&Multiply_Float(A=$Emitter.LoopedAge, B=1)
       )
       Period=1
       Scale=&FloatFromCurve(FloatCurve=DataInterface() {propertyValues {Curve {
           defaultValue=1
           keys [{arriveTangent=4.6832, interpMode=RCIM_Cubic, leaveTangent=4.6832, tangentMode=RCTM_User, time=0, value=0}, <at 0.5 1 RCIM_Cubic>]
          }}})
       Bias=0
      )
     )
     Passthrough Vector=&MakeVector(
      X=-1
      Y=0
      Z=&Sine(
       Normalized Angle=&Add_Float(
        A=&Multiply_Float(A=$Particles.NormalizedAge, B=1)
        B=&Multiply_Float(A=$Emitter.LoopedAge, B=1)
       )
       Period=1
       Scale=&FloatFromCurve(FloatCurve=DataInterface() {propertyValues {Curve {
           defaultValue=1
           keys [{arriveTangent=4.6832, interpMode=RCIM_Cubic, leaveTangent=4.6832, tangentMode=RCTM_User, time=0, value=0}, <at 0.5 1 RCIM_Cubic>]
          }}})
       Bias=0
      )
     )
    )) {isSetParameters=true, module=SetVariables_#6}
   /SetVariables(
    Particles.linescolor:LinearColor=&MultiplyLinearColor(
     Linear Color=<rgba 0.5543 0.6685 0.7826 1>
     Scale Factor=&MakeVector4FromVectorAndFloat(
      "Vector (RGB)"=&VectorFromFloat(Value=&MergeHelper_FloatSwitch(
        0=2
        1=1
        2=?
        3=?
        Index=&MakeCustomIntFromBool(Bool=$User.bIsFrontendPreview, True INT=1, False INT=0)
        Num=::NewEnumerator0
       ))
      "Float (Alpha)"=&FloatFromCurve(
       FloatCurve=DataInterface() {propertyValues {Curve {
          defaultValue=1
          keys [
           {arriveTangent=0.0979, interpMode=RCIM_Cubic, leaveTangent=0.0979, tangentMode=RCTM_User, time=0.05, value=0}
           {interpMode=RCIM_Cubic, tangentMode=RCTM_User, time=0.3, value=1}
          ]
         }}}
       Scale Curve=4
      )
     )
    )
    Particles.logocolor:LinearColor=&MultiplyLinearColor(
     Linear Color=<rgba 0.2908 0.3074 0.324 1>
     Scale Factor=&MakeVector4FromVectorAndFloat(
      "Vector (RGB)"=&VectorFromFloat(Value=&MergeHelper_FloatSwitch(
        0=2
        1=1
        2=?
        3=?
        Index=&MakeCustomIntFromBool(Bool=$User.bIsFrontendPreview, True INT=1, False INT=0)
        Num=::NewEnumerator0
       ))
      "Float (Alpha)"=1
     )
    )
    Particles.scarfcolor:LinearColor=&MultiplyLinearColor(
     Linear Color=<rgba 0.011 0.0508 0.5043 1>
     Scale Factor=&MakeVector4FromVectorAndFloat(
      "Vector (RGB)"=&VectorFromFloat(Value=&MergeHelper_FloatSwitch(
        0=4
        1=2
        2=?
        3=?
        Index=&MakeCustomIntFromBool(Bool=$User.bIsFrontendPreview, True INT=1, False INT=0)
        Num=::NewEnumerator0
       ))
      "Float (Alpha)"=&FloatFromCurve(FloatCurve=DataInterface() {propertyValues {Curve {
          defaultValue=1
          keys [
           {arriveTangent=0.0979, interpMode=RCIM_Cubic, leaveTangent=0.0979, tangentMode=RCTM_User, time=0.05, value=0}
           {interpMode=RCIM_Cubic, tangentMode=RCTM_User, time=0.15, value=1}
          ]
         }}})
     )
    )
   ) {isSetParameters=true}
  ]
  rendererClasses [&NiagaraRibbonRendererProperties]
  renderers [
   {rendererClass=&NiagaraRibbonRendererProperties}
   {rendererClass=&NiagaraRibbonRendererProperties, rendererIndex=1}
   {rendererClass=&NiagaraRibbonRendererProperties, rendererIndex=2}
  ]
 }
 Scarf_R {
  emitterUpdateScript [
   &EmitterState()
   &VC_FilteredSocketCount()
   &VC_RateQualityHandler(Epic Spawn Rate=40, High Spawn Rate=40, Low Spawn Rate=30, Use Custom Rate?=true)
   &VC_SpawnRibbon_v1(SpawnRate=$Emitter.Spawn Rate)
  ]
  particleSpawnScript [
   &VC_PaintedVariantsOverride() {enabled=false}
   &VC_InitializeRibbon_v1(
    Custom Offset=true
    Enable Skeletal Position=true
    Lifetime=0.8
    Local Vector Offset=&AddVectorToPosition(Position=<20 -10 0>, Vector=$User.DTCarOffset)
    Offset Amount=1
    Ribbon Width=30
   )
   &AddVelocity(Velocity=&MergeHelper_VectorSwitch(0=<-1500 0 0>, 1=<-1500 0 0>, 2=<-1200 0 0>, 3=?, Index=$User.GameModeID, Num=::NewEnumerator1)) {module=AddVelocity001}
   /SetVariables() {isSetParameters=true, module=SetVariables_#2}
   /SetVariables(Particles.Position=&AddVectorToPosition(
     Position=$Particles.Position
     Vector=&TransformVector(
      Vector=&MakeVector(X=0, Y=&MaskFloatBySpawnGroup(Masked Float=10, Passthrough Float=-10), Z=0)
      Destination Space=::Simulation
     )
    )) {isSetParameters=true, module=SetVariables_#3}
  ]
  particleUpdateScript [
   &ParticleState()
   &InheritSourceMovement(ApplyForce=false, ApplyPosition=true, ApplyPositionFalloff=false)
   &SolveForcesAndVelocity(Clamp Velocity=false, Limit Acceleration=false)
   &VC_UpdateRibbon_v1(Detach Locked Particle=true, Enable Skeletal Position=true)
   /SetVariables(Particles.Position=&AddVectorToPosition(
     Position=$Particles.Position
     Vector=&VC_ConvertVectorToWorld(Vector:Vector3f=&MakeVector(
       X=0
       Y=0
       Z=&FloatFromCurve(
        FloatCurve=DataInterface() {propertyValues {Curve {keys [
            <at 0 0 RCIM_Cubic>
            {arriveTangent=103.5191, interpMode=RCIM_Cubic, leaveTangent=103.5191, tangentMode=RCTM_User, time=0.22, value=40}
            {arriveTangent=3.1163, interpMode=RCIM_Cubic, leaveTangent=3.1163, tangentMode=RCTM_User, time=0.5, value=40}
            {arriveTangent=-14.8689, interpMode=RCIM_Cubic, leaveTangent=-14.8689, tangentMode=RCTM_User, time=0.8, value=0}
           ]}}}
        Scale Curve=-0.05
       )
      ))
    )) {isSetParameters=true, module=SetVariables_#4}
   /SetVariables(Particles.Position=&AddVectorToPosition(
     Position=$Particles.Position
     Vector=&TransformVector(
      Vector=&MakeVector(
       X=0
       Y=&Sine(
        Normalized Angle=&Add_Float(
         A=&Multiply_Float(A=$Particles.NormalizedAge, B=1.4)
         B=&Multiply_Float(A=$Emitter.LoopedAge, B=1)
        )
        Period=1
        Scale=&FloatFromCurve(
         FloatCurve=DataInterface() {propertyValues {Curve {defaultValue=1, keys [{tangentMode=RCTM_User, time=0, value=0}, {tangentMode=RCTM_User, time=0.53, value=1.1}]}}}
         Scale Curve=17
        )
        Bias=0
       )
       Z=&Sine(
        Normalized Angle=&Add_Float(
         A=&Multiply_Float(A=$Particles.NormalizedAge, B=-1.6)
         B=&Multiply_Float(A=$Emitter.LoopedAge, B=1.2)
        )
        Period=1
        Scale=&FloatFromCurve(
         FloatCurve=DataInterface() {propertyValues {Curve {
            defaultValue=1
            keys [
             {tangentMode=RCTM_User, time=0.1, value=0.3}
             {arriveTangent=-0.2778, interpMode=RCIM_Cubic, leaveTangent=-3.8301, tangentMode=RCTM_Break, time=0.3, value=1.2}
             <at 1 0.05 RCIM_Cubic>
            ]
           }}}
         Scale Curve=4
        )
        Bias=0
       )
      )
      Destination Space=::Simulation
      Apply Scale=true
     )
    )) {isSetParameters=true, module=SetVariables_#5}
   /SetVariables(Particles.RibbonFacing=&MaskVectorBySpawnGroup(
     Masked Vector=&MakeVector(
      X=1
      Y=0
      Z=&Sine(
       Normalized Angle=&Add_Float(
        A=&Multiply_Float(A=$Particles.NormalizedAge, B=1)
        B=&Multiply_Float(A=$Emitter.LoopedAge, B=1)
       )
       Period=1
       Scale=&FloatFromCurve(FloatCurve=DataInterface() {propertyValues {Curve {
           defaultValue=1
           keys [{arriveTangent=4.6832, interpMode=RCIM_Cubic, leaveTangent=4.6832, tangentMode=RCTM_User, time=0, value=0}, <at 0.5 1 RCIM_Cubic>]
          }}})
       Bias=0
      )
     )
     Passthrough Vector=&MakeVector(
      X=-1
      Y=0
      Z=&Sine(
       Normalized Angle=&Add_Float(
        A=&Multiply_Float(A=$Particles.NormalizedAge, B=1)
        B=&Multiply_Float(A=$Emitter.LoopedAge, B=1)
       )
       Period=1
       Scale=&FloatFromCurve(FloatCurve=DataInterface() {propertyValues {Curve {
           defaultValue=1
           keys [{arriveTangent=4.6832, interpMode=RCIM_Cubic, leaveTangent=4.6832, tangentMode=RCTM_User, time=0, value=0}, <at 0.5 1 RCIM_Cubic>]
          }}})
       Bias=0
      )
     )
    )) {isSetParameters=true, module=SetVariables_#6}
   /SetVariables(
    Particles.linescolor:LinearColor=&MultiplyLinearColor(
     Linear Color=<rgba 0.5543 0.6685 0.7826 1>
     Scale Factor=&MakeVector4FromVectorAndFloat(
      "Vector (RGB)"=&VectorFromFloat(Value=&MergeHelper_FloatSwitch(
        0=2
        1=1
        2=?
        3=?
        Index=&MakeCustomIntFromBool(Bool=$User.bIsFrontendPreview, True INT=1, False INT=0)
        Num=::NewEnumerator0
       ))
      "Float (Alpha)"=&FloatFromCurve(
       FloatCurve=DataInterface() {propertyValues {Curve {
          defaultValue=1
          keys [
           {arriveTangent=0.0979, interpMode=RCIM_Cubic, leaveTangent=0.0979, tangentMode=RCTM_User, time=0.05, value=0}
           {interpMode=RCIM_Cubic, tangentMode=RCTM_User, time=0.3, value=1}
          ]
         }}}
       Scale Curve=4
      )
     )
    )
    Particles.logocolor:LinearColor=&MultiplyLinearColor(
     Linear Color=<rgba 0.2908 0.3074 0.324 1>
     Scale Factor=&MakeVector4FromVectorAndFloat(
      "Vector (RGB)"=&VectorFromFloat(Value=&MergeHelper_FloatSwitch(
        0=2
        1=1
        2=?
        3=?
        Index=&MakeCustomIntFromBool(Bool=$User.bIsFrontendPreview, True INT=1, False INT=0)
        Num=::NewEnumerator0
       ))
      "Float (Alpha)"=1
     )
    )
    Particles.scarfcolor:LinearColor=&MultiplyLinearColor(
     Linear Color=<rgba 0.011 0.0508 0.5043 1>
     Scale Factor=&MakeVector4FromVectorAndFloat(
      "Vector (RGB)"=&VectorFromFloat(Value=&MergeHelper_FloatSwitch(
        0=4
        1=2
        2=?
        3=?
        Index=&MakeCustomIntFromBool(Bool=$User.bIsFrontendPreview, True INT=1, False INT=0)
        Num=::NewEnumerator0
       ))
      "Float (Alpha)"=&FloatFromCurve(FloatCurve=DataInterface() {propertyValues {Curve {
          defaultValue=1
          keys [
           {arriveTangent=0.0979, interpMode=RCIM_Cubic, leaveTangent=0.0979, tangentMode=RCTM_User, time=0.05, value=0}
           {interpMode=RCIM_Cubic, tangentMode=RCTM_User, time=0.15, value=1}
          ]
         }}})
     )
    )
   ) {isSetParameters=true}
  ]
  rendererClasses [&NiagaraRibbonRendererProperties]
  renderers [
   {rendererClass=&NiagaraRibbonRendererProperties}
   {rendererClass=&NiagaraRibbonRendererProperties, rendererIndex=1}
   {rendererClass=&NiagaraRibbonRendererProperties, rendererIndex=2}
  ]
 }
 Snow {
  emitterUpdateScript [
   &EmitterState()
   &VC_FilteredSocketCount()
   &SpawnRate(SpawnRate=160, Use Spawn Probability=false)
   &SpawnRate(SpawnGroup=1, SpawnRate=60, Use Spawn Probability=false) {module=SpawnRate001}
  ]
  particleSpawnScript [
   &VC_PaintedVariantsOverride() {enabled=false}
   &InitializeParticle(
    Color=<rgba 0.46 0.775 1 0.04>
    Lifetime Max=0.9
    Lifetime Min=0.75
    Lifetime Mode=::NewEnumerator1
    Sprite Rotation Mode=::NewEnumerator2
    Sprite Size Mode=::NewEnumerator1
    Uniform Sprite Size Max=130
    Uniform Sprite Size Min=110
    UsePositionOffset=false
    Write Material Random=true
   )
   &PartitionParticles(Number Of Partitions=$Emitter.PortCount)
   &SkeletalMeshLocation(
    Bone / Socket Index=$Particles.PartitionParticles.Partition
    Bone / Socket Sampling Mode=::NewEnumerator5
    MaskSpawnGroup=false
    Mesh Sampling Type=::NewEnumerator6
    Skeletal Mesh=$System.SkeletalMesh
   )
   /SetVariables(Particles.snowsize:Vector2f=&RandomRangeVector2D(Minimum=<80 80>, Maximum=<110 110>)) {isSetParameters=true}
   &ShapeLocation(
    MaskSpawnGroup=false
    Non Uniform Scale=<1 1.2 1.1>
    Offset=$User.DTCarOffset
    Offset Coordinate Space=::Local
    Offset Mode=::NewEnumerator0
    Override Seed=false
    Sphere Radius=20
    UseOwnerScale=true
   )
   &AddVelocity(
    Velocity=&MergeHelper_VectorSwitch(
     0=&RandomRangeVector(Minimum=<-400 -15 -15>, Maximum=<-900 15 15>)
     1=&RandomRangeVector(Minimum=<-500 -10 10>, Maximum=<-700 10 10>)
     2=?
     3=?
     Index=&SpawnGroup()
     Num=::NewEnumerator0
    )
    Velocity Speed Scale=&MergeHelper_FloatSwitch(
     0=0.6
     1=1
     2=?
     3=?
     Index=&MakeCustomIntFromBool(Bool=$User.bIsFrontendPreview, True INT=1, False INT=0)
     Num=::NewEnumerator0
    )
   )
   /SetVariables(Particles.VisibilityTag=$Emitter.SpawnGroup) {isSetParameters=true}
  ]
  particleUpdateScript [
   &ParticleState()
   &SolveForcesAndVelocity(Clamp Velocity=false, Limit Acceleration=false)
   &ScaleColor(
    Scale Alpha=&Multiply_Float(
     A=&FloatFromCurve(FloatCurve=DataInterface() {propertyValues {Curve {
         defaultValue=1
         keys [
          {interpMode=RCIM_Cubic, leaveTangentWeight=0.1667, tangentMode=RCTM_User, time=0, value=0.5}
          {arriveTangentWeight=0.1667, interpMode=RCIM_Cubic, leaveTangentWeight=0.1667, tangentMode=RCTM_User, time=0.5043, value=0.8771}
          {arriveTangentWeight=0.1667, interpMode=RCIM_Cubic, tangentMode=RCTM_User, time=0.95, value=0}
         ]
        }}})
     B=&MergeHelper_FloatSwitch(
      0=2
      1=1
      2=?
      3=?
      Index=&MakeCustomIntFromBool(Bool=$User.bIsFrontendPreview, True INT=1, False INT=0)
      Num=::NewEnumerator0
     )
    )
    Scale RGB=&VectorFromFloat(Value=&Multiply_Float(
      A=&FloatFromCurve(FloatCurve=DataInterface() {propertyValues {Curve {
          defaultValue=1
          keys [
           {interpMode=RCIM_Cubic, leaveTangentWeight=0.1667, tangentMode=RCTM_User, time=0, value=0.8}
           {
            arriveTangent=0.1311
            arriveTangentWeight=0.1304
            interpMode=RCIM_Cubic
            leaveTangent=0.1311
            leaveTangentWeight=0.0422
            tangentMode=RCTM_User
            time=0.3995
            value=0.8152
           }
           {arriveTangentWeight=0.0408, interpMode=RCIM_Cubic, leaveTangentWeight=0.0481, tangentMode=RCTM_User, time=0.5007, value=0.8542}
           {
            arriveTangent=-0.0618
            arriveTangentWeight=0.0497
            interpMode=RCIM_Cubic
            leaveTangent=-0.0618
            leaveTangentWeight=0.1227
            tangentMode=RCTM_User
            time=0.6048
            value=0.8123
           }
           {arriveTangentWeight=0.1667, interpMode=RCIM_Cubic, tangentMode=RCTM_User, time=1, value=0.8}
          ]
         }}})
      B=&MergeHelper_FloatSwitch(
       0=2
       1=1
       2=?
       3=?
       Index=&MakeCustomIntFromBool(Bool=$User.bIsFrontendPreview, True INT=1, False INT=0)
       Num=::NewEnumerator0
      )
     ))
    ScaleA=true
    ScaleRGB=true
   )
   /SetVariables(Particles.snowcolor:LinearColor=&MultiplyLinearColor(
     Linear Color=<rgba 0.51 0.8243 1 1>
     Scale Factor=&MakeVector4FromVectorAndFloat(
      "Vector (RGB)"=&VectorFromFloat(Value=&Multiply_Float(
        A=&MergeHelper_FloatSwitch(
         0=20
         1=15
         2=?
         3=?
         Index=&MakeCustomIntFromBool(Bool=$User.bIsFrontendPreview, True INT=1, False INT=0)
         Num=::NewEnumerator0
        )
        B=&FloatFromCurve(FloatCurve=DataInterface() {propertyValues {Curve {
            defaultValue=1
            keys [
             {interpMode=RCIM_Cubic, leaveTangentWeight=0.1, tangentMode=RCTM_User, time=0, value=0.9}
             {arriveTangentWeight=0.1008, interpMode=RCIM_Cubic, leaveTangentWeight=0.168, tangentMode=RCTM_User, time=0.25, value=1}
             {arriveTangentWeight=0.2906, interpMode=RCIM_Cubic, leaveTangentWeight=0.1163, tangentMode=RCTM_User, time=0.9, value=1}
             {arriveTangent=-2.7695, arriveTangentWeight=0.0667, interpMode=RCIM_Cubic, leaveTangent=-2.7695, tangentMode=RCTM_User, time=1, value=0}
            ]
           }}})
       ))
      "Float (Alpha)"=&Multiply_Float(
       A=&MergeHelper_FloatSwitch(
        0=15
        1=13
        2=?
        3=?
        Index=&MakeCustomIntFromBool(Bool=$User.bIsFrontendPreview, True INT=1, False INT=0)
        Num=::NewEnumerator0
       )
       B=&FloatFromCurve(FloatCurve=DataInterface() {propertyValues {Curve {
           defaultValue=1
           keys [
            {interpMode=RCIM_Cubic, tangentMode=RCTM_User, time=0, value=0.8}
            {arriveTangent=0.2353, interpMode=RCIM_Cubic, leaveTangent=0.2353, time=0.25, value=1}
            {arriveTangentWeight=0.2134, interpMode=RCIM_Cubic, leaveTangentWeight=0.2134, tangentMode=RCTM_User, time=0.85, value=1}
            {arriveTangent=-2.9976, arriveTangentWeight=0.1667, interpMode=RCIM_Cubic, leaveTangent=-2.9976, tangentMode=RCTM_User, time=1, value=0}
           ]
          }}})
      )
     )
    )) {isSetParameters=true}
  ]
  rendererClasses [&NiagaraSpriteRendererProperties]
  renderers [
   {rendererClass=&NiagaraSpriteRendererProperties}
   {rendererClass=&NiagaraSpriteRendererProperties, rendererIndex=1}
  ]
 }
}
summary {
 systemName=NS_NumbLTE03
 emitters [
  {emitterName=Scarf_L, rendererClasses [&NiagaraRibbonRendererProperties]}
  {emitterName=Scarf_R, rendererClasses [&NiagaraRibbonRendererProperties]}
  {emitterName=Glow, rendererClasses [&NiagaraSpriteRendererProperties]}
  {emitterName=Snow, rendererClasses [&NiagaraSpriteRendererProperties]}
 ]
 userVariables [
  User.bIsFrontendPreview {type=NiagaraBool, defaultValue=true}
  User.bIsPainted {type=NiagaraBool}
  User.ColorVariantID {type=NiagaraInt32}
  User.DTCarOffset {type=Vector3f, description="", defaultValue {struct=Vector3f, value=<0 0 0>}}
  User.GameModeID {type=NiagaraInt32}
  User.PaintedColor {type=LinearColor, description="", defaultValue {struct=LinearColor, value=<rgba 1 1 1 1>}}
  User.PaintedScalar {type=NiagaraFloat, defaultValue=1, description=""}
  User.PortCount {type=NiagaraInt32, defaultValue=2}
 ]
}
